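# Post-process AnimBP for the F_MED_ClawedRaven head cosmetic, adding secondary motion to its hood.
# The AnimGraph offsets hood_R additively in bone space, then chains four root-relative Anim Dynamics nodes (top back, top front, left, right) using angular springs, limits and reduced gravity.
# The EventGraph caches the body mesh on customization load and maps right upper-arm yaw relative to the clavicle into a variable, which the AnimGraph does not visibly read.
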AnimBlueprint /BRCosmetics/Characters/Player/Female/Medium/Heads/F_MED_ClawedRaven_Head/Meshes/F_MED_ClawedRaven_Head_PostProcessBP
target_skeleton=&/F_MED_ClawedRaven_Head_Skeleton
graphs [
 AnimGraph {nodes [
   n1 = "Transform (Modify) Bone - Bone: hood_R"!(ComponentPose=Local To Component(LocalPose=Input Pose()), Alpha=1, Translation="0,0,-13", Rotation="0,0,0") -> (Pose) {
    translationMode=BMM_Additive
    translationSpace=BCS_BoneSpace
    alphaBoolBlend {blendOption=Linear}
    alphaScaleBiasClamp {bInterpResult=true, interpSpeedDecreasing=5, interpSpeedIncreasing=20}
    boneToModify {boneName=hood_R}
    translation {z=-13}
   }
   n2 = Anim Dynamics - Bone: dyn_hood_top_bk_01!(ComponentPose=@n1, Alpha=1, GravityOverride:Vector="0,0,-980") -> (Pose) {
    angularBiasOverride=1
    angularDampingOverride=1
    angularSpringConstant=1000
    bAngularSpring=true
    bOverrideAngularBias=true
    bOverrideAngularDamping=true
    bUseGravityOverride=true
    bUsePlanarLimit=false
    gravityScale=0.35
    simulationSpace=RootRelative
    alphaBoolBlend {blendOption=Linear}
    boundBone {boneName=dyn_hood_top_bk_01}
    componentAppliedLinearAccClamp=<1000 1000 1000>
    componentLinearAccScale=<0.7 0.7 0.7>
    componentLinearVelScale=<0.7 0.7 0.7>
    gravityOverride {z=-980}
    physicsBodyDefinitions [{
      sphereCollisionRadius=2
      boundBone {boneName=dyn_hood_top_bk_01}
      boxExtents=<14 14 14>
      constraintSetup {angularLimitsMax=<0 45 10>, angularLimitsMin=<0 -20 -10>, angularTarget=<1 0 0>}
      localJointOffset=<0.5 0 -1.3>
     }]
   }
   n3 = Anim Dynamics - Bone: dyn_hood_top_fr_01!(ComponentPose=@n2, Alpha=1, GravityOverride:Vector="0,0,-980") -> (Pose) {
    angularBiasOverride=0.9995
    angularDampingOverride=0.9995
    angularSpringConstant=1000
    bAngularSpring=true
    bOverrideAngularBias=true
    bOverrideAngularDamping=true
    bUseGravityOverride=true
    bUsePlanarLimit=false
    gravityScale=0.35
    simulationSpace=RootRelative
    alphaBoolBlend {blendOption=Linear}
    boundBone {boneName=dyn_hood_top_fr_01}
    componentAppliedLinearAccClamp=<1000 1000 1000>
    componentLinearAccScale=<0.7 0.7 0.7>
    componentLinearVelScale=<0.7 0.7 0.7>
    gravityOverride {z=-980}
    physicsBodyDefinitions [{
      sphereCollisionRadius=2
      boundBone {boneName=dyn_hood_top_fr_01}
      boxExtents=<9.5 9.5 9.5>
      constraintSetup {angularLimitsMax=<0 25 10>, angularLimitsMin=<0 -10 -10>, angularTarget=<1 0 0>}
      localJointOffset=<0.1 0 0>
     }]
   }
   n4 = Anim Dynamics - Bone: dyn_hood_01_L!(ComponentPose=@n3, Alpha=1, GravityOverride:Vector="0,0,-980") -> (Pose) {
    angularBiasOverride=0.999
    angularDampingOverride=0.999
    angularSpringConstant=1000
    bAngularSpring=true
    bOverrideAngularBias=true
    bOverrideAngularDamping=true
    bUseGravityOverride=true
    bUsePlanarLimit=false
    gravityScale=0.35
    simulationSpace=RootRelative
    alphaBoolBlend {blendOption=Linear}
    boundBone {boneName=dyn_hood_01_L}
    componentAppliedLinearAccClamp=<1000 1000 1000>
    componentLinearAccScale=<0.7 0.7 0.7>
    componentLinearVelScale=<0.7 0.7 0.7>
    gravityOverride {z=-980}
    physicsBodyDefinitions [{
      sphereCollisionRadius=2
      boundBone {boneName=dyn_hood_01_L}
      boxExtents=<13 13 13>
      constraintSetup {angularLimitsMax=<0 10 10>, angularLimitsMin=<0 -25 -10>, angularTarget=<1 0 0>}
      localJointOffset=<0.5 0 0>
     }]
   }
   n5 = Anim Dynamics - Bone: dyn_hood_01_R!(ComponentPose=@n4, Alpha=1, GravityOverride:Vector="0,0,-980") -> (Pose) {
    angularBiasOverride=0.999
    angularDampingOverride=0.999
    angularSpringConstant=1000
    bAngularSpring=true
    bOverrideAngularBias=true
    bOverrideAngularDamping=true
    bUseGravityOverride=true
    bUsePlanarLimit=false
    gravityScale=0.35
    simulationSpace=RootRelative
    alphaBoolBlend {blendOption=Linear}
    boundBone {boneName=dyn_hood_01_R}
    componentAppliedLinearAccClamp=<1000 1000 1000>
    componentLinearAccScale=<0.7 0.7 0.7>
    componentLinearVelScale=<0.7 0.7 0.7>
    gravityOverride {z=-980}
    physicsBodyDefinitions [{
      sphereCollisionRadius=2
      boundBone {boneName=dyn_hood_01_R}
      boxExtents=<13 13 13>
      constraintSetup {angularLimitsMax=<0 10 10>, angularLimitsMin=<0 -25 -10>, angularTarget=<1 0 0>}
      localJointOffset=<0.5 0 0>
     }]
   }
   Output Pose(Result=Component To Local(ComponentPose=@n5))
  ]}
 EventGraph {nodes [
   Event BlueprintUpdateAnimation()
   n1 = BreakTransform(InTransform=GetDeltaTransformFromRefPose(self=Get BodyMesh() -> (BodyMesh:Skeletal Mesh Component Object Reference), BoneName=upperarm_r, BaseName=clavicle_r))
   > Set upperarm r rot yaw(upperarm r rot yaw:float=MapRangeClamped(Value=BreakRotator(InRot=@n1.Rotation).Yaw, InRangeA=90, InRangeB=110)) -> (Output_Get:float)
   n2 = Event OnCharacterCustomizationLoaded()
   > Set BodyMesh(BodyMesh=Get Array Item with EFort Custom Part Type(Array=@n2.CustomMeshParts))
   TryGetPawnOwner()
  ]}
]
variables [
 BodyMesh {type=SkeletalMeshComponent}
 upperarm r rot yaw {type=number}
]
